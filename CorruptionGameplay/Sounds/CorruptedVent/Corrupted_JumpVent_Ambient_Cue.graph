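# Sound cue for the looping ambient sound of a corrupted jump vent in the CorruptionGameplay content.
# A single looping wave player feeds the Corrupted_JumpVent_Ambient wave straight to the output at reduced volume.
# A corrupted-vent attenuation asset spatializes it over a limited range, and it routes through the scripted-events sound class.
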SoundCue /CorruptionGameplay/Sounds/CorruptedVent/Corrupted_JumpVent_Ambient_Cue
root=WavePlayer_0
nodes [{id=WavePlayer_0, properties {bLooping=true, soundWaveAssetPtr=&/Corrupted_JumpVent_Ambient}}]
properties {
 attenuationSettings=&/Fort_CorruptedVent_Att
 maxDistance=2600
 soundClassObject=/Game/Sounds/FortSoundClassesAndModes/Fort_Events_Scripted
 volumeMultiplier=0.2
}
