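# Sound cue for the launch of a corrupted jump vent, layering a corruption-specific sound with a reused sideways jump pad launch.
# A mixer plays one of three random corrupted vent waves immediately at full volume.
# It also plays one of three random jump pad waves at half volume after a short random delay.
# The cue uses corrupted vent attenuation and the scripted events sound class.
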SoundCue /CorruptionGameplay/Sounds/CorruptedVent/Corrupted_JumpVent_Launch_Cue
root=Mixer_0
nodes [
 {id=Delay_0, children [Random_1], properties {delayMax=0.15, delayMin=0.1}}
 {id=Mixer_0, children [Delay_0,Random_0], properties {inputVolume [0.5,1]}}
 {id=Random_0, children [WavePlayer_0,WavePlayer_1,WavePlayer_2], properties {weights [1,1,1]}}
 {id=Random_1, children [WavePlayer_3,WavePlayer_4,WavePlayer_5], properties {weights [1,1,1]}}
 {id=WavePlayer_0, properties {soundWaveAssetPtr=&/Corrupted_JumpVent_Launch_01}}
 {id=WavePlayer_1, properties {soundWaveAssetPtr=&/Corrupted_JumpVent_Launch_02}}
 {id=WavePlayer_2, properties {soundWaveAssetPtr=&/Corrupted_JumpVent_Launch_03}}
 {id=WavePlayer_3, properties {soundWaveAssetPtr=&/Sideways_JumpPad_Launch_01}}
 {id=WavePlayer_4, properties {soundWaveAssetPtr=&/Sideways_JumpPad_Launch_02}}
 {id=WavePlayer_5, properties {soundWaveAssetPtr=&/Sideways_JumpPad_Launch_03}}
]
properties {
 attenuationSettings=&/Fort_CorruptedVent_Att
 maxDistance=2600
 soundClassObject=/Game/Sounds/FortSoundClassesAndModes/Fort_Events_Scripted
}
