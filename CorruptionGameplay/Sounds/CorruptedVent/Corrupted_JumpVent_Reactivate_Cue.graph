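# Sound cue for a corrupted jump vent reactivating in the CorruptionGameplay content.
# A Random node picks one of three reactivate wave variations with equal weight.
# Playback uses the corrupted vent attenuation asset and the scripted events sound class, at a reduced volume multiplier.
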SoundCue /CorruptionGameplay/Sounds/CorruptedVent/Corrupted_JumpVent_Reactivate_Cue
root=Random_0
nodes [
 {id=Random_0, children [WavePlayer_0,WavePlayer_1,WavePlayer_2], properties {weights [1,1,1]}}
 {id=WavePlayer_0, properties {soundWaveAssetPtr=&/Corrupted_JumpVent_Reactivate_03}}
 {id=WavePlayer_1, properties {soundWaveAssetPtr=&/Corrupted_JumpVent_Reactivate_02}}
 {id=WavePlayer_2, properties {soundWaveAssetPtr=&/Corrupted_JumpVent_Reactivate_01}}
]
properties {
 attenuationSettings=&/Fort_CorruptedVent_Att
 maxDistance=2600
 soundClassObject=/Game/Sounds/FortSoundClassesAndModes/Fort_Events_Scripted
 volumeMultiplier=0.6
}
